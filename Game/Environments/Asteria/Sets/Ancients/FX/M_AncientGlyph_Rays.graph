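# Additive, unlit, two-sided material for Ancient glyph light rays on Niagara mesh particles.
# Opacity combines particle alpha, a gradient, a UV falloff, two tile-pattern samples animated by RaysMove and double-sided Fresnel, with depth fade on High and Epic.
# Particle color feeds emissive adjustments.
# A PushOut offset blends particle-radial and vertex-normal directions, scaled by RaysDir and tapered along selectable, swizzlable UVs.
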Material /Game/Environments/Asteria/Sets/Ancients/FX/M_AncientGlyph_Rays
graph {
 properties {
  bUsedWithNiagaraMeshParticles=true
  blendMode=BLEND_Additive
  pixelDepthOffsetMode=PDOM_Legacy
  shadingModel=MSM_Unlit
  twoSided=true
  usedShadingModels=Unlit
 }
 nodes [
  n1 = ParticleColor()
  MaterialFunctionCall(Emissive=@n1.RGB) {materialFunction=&MF_FX_EmissiveAdjustments}
  n2 = NamedRerouteUsage()
  n3 = WorldPosition()
  n4 = ParticlePositionWS()
  n5 = LinearInterpolate(A=Normalize(VectorInput=Subtract(A=@n3.XYZ, B=@n4)), B=VertexNormalWS(), Alpha=ScalarParameter() {parameterName=PushAlongVertexNormal})
  n6 = VectorParameter() {parameterName=RaysDir, defaultValue=<rgba 1 0 1 1>}
  n7 = Normalize(VectorInput=@n6.RGB)
  Multiply(
   A=OneMinus(None=ComponentMask(None=@n2) {r=true})
   B=Multiply(A=Multiply(A=@n5, B=@n7), B=DynamicParameter() {paramNames [RaysTiling,RaysMove,PushOut,Param4]}.PushOut)
  )
  n8 = LinearInterpolate(
   A=TextureCoordinate()
   B=MaterialFunctionCall() {materialFunction=&Swizzle}.YX
   Alpha=ScalarParameter() {parameterName=SwizzleUV}
  )
  n9 = VectorParameter() {parameterName=Tiling, defaultValue=<rgba 1 1 0 1>}
  NamedRerouteDeclaration(None=Multiply(A=NamedRerouteDeclaration(None=@n8) {name=UV Picker, nodeColor=<rgba 0.5234 1 0 1>}, B=ComponentMask(None=@n9.RGB) {g=true, r=true})) {name=UVPickerWithTile, nodeColor=<rgba 0.9688 0 1 1>}
  n10 = MaterialFunctionCall(XY=NamedRerouteUsage()) {materialFunction=&Swizzle}
  n11 = Multiply(
   A=@n1.A
   B=TextureSample(UVs=OneMinus(None=@n10.YX)) {samplerType=SAMPLERTYPE_LinearGrayscale, texture=/Game/Effects/Fort_Effects/Textures/Grad/T_Vertical_Gradient_Wave}.G
  )
  n12 = Power(Base=ComponentMask(None=NamedRerouteUsage()) {r=true}) {constExponent=4}
  n13 = DynamicParameter() {paramNames [RaysTiling,RaysMove,PushOut,Param4]}
  n14 = NamedRerouteUsage()
  n15 = ComponentMask(None=@n14) {g=true}
  n16 = TextureSample(UVs=AppendVector(A=@n13.RaysMove, B=Multiply(A=@n15, B=@n13.RaysTiling))) {texture=/Game/Effects/Fort_Effects/Textures/T_Tile_Pattern_01}
  n17 = Multiply(A=@n14) {constB=2}
  n18 = TextureSample(UVs=AppendVector(A=Multiply(A=@n13.RaysMove) {constB=0.5}, B=ComponentMask(None=@n17) {g=true})) {texture=&T_Tile_Pattern_01}
  n19 = Multiply(
   A=Multiply(A=Multiply(A=@n11, B=@n12), B=Add(A=@n16.R, B=@n18.R))
   B=MaterialFunctionCall(Exponent=ScalarParameter() {parameterName=FresnelExp}) {materialFunction=&MF_Fresnel_FX}.Fresnel Double Sided
  )
  n20 = DepthFade(Opacity=@n19, FadeDistance=ScalarParameter() {defaultValue=50, parameterName=DepthFadeDistance}) {fadeDistanceDefault=50}
  QualitySwitch(Default=@n19, Low=@n19, High=@n20, Medium=@n19, Epic=@n20)
  Transform(None=@n6.RGB) {transformSourceType=TRANSFORMSOURCE_Local}
 ]
}
