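# Additive unlit Niagara sprite material for a glowing Ancient glyph effect, shaped by the blue channel of a mask texture.
# Radially mapped, panning plasma distorts a second plasma sample, adding energy to the mask and sphere-falloff brightness.
# Particle color blends with a falloff color, while particle alpha and depth fade control intensity.
# Dynamic parameters drive brightness, energy and bump offset, and bIsModulate switches to a fade-from-white variant.
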Material /Game/Environments/Asteria/Sets/Ancients/FX/M_Ch4s3_AncientGlyph_Illum
graph {
 properties {
  bUsedWithNiagaraSprites=true
  blendMode=BLEND_Additive
  pixelDepthOffsetMode=PDOM_Legacy
  shadingModel=MSM_Unlit
  translucencyPass=MTP_BeforeDOF
  usedShadingModels=Unlit
 }
 nodes [
  n1 = Subtract(
   A=MaterialFunctionCall() {materialFunction=&WorldPositionBehindTranslucency}.Result
   B=ActorPositionWS()
  )
  n2 = Divide(A=ComponentMask(None=TransformPosition(None=@n1) {transformSourceType=TRANSFORMPOSSOURCE_World, transformType=TRANSFORMPOSSOURCE_Instance}) {b=true, r=true}) {constB=1000}
  n3 = Constant2Vector() {r=0.5}
  Add(
   A=Multiply(A=Subtract(A=@n2, B=@n3) {constB=0.5}, B=Constant2Vector() {g=-1, r=-1})
   B=ComponentMask(None=VectorParameter() {parameterName=UV_Offset, defaultValue=<rgba 0 0 0 1>}.RGB) {g=true, r=true}
  )
  n4 = ParticleColor()
  n5 = LinearInterpolate(A=VectorParameter() {parameterName=FalloffColor}.RGB, B=@n4.RGB, Alpha=NamedRerouteUsage())
  n6 = Add(A=NamedRerouteUsage(), B=NamedRerouteUsage())
  n7 = Multiply(A=MaterialFunctionCall() {materialFunction=&VectorToRadialValue}.Radial Coordinates) {constB=4}
  n8 = Panner(Coordinate=@n7) {bFractionalPart=true, speedY=0.01}
  n9 = Add(
   A=Panner(Coordinate=@n7) {bFractionalPart=true, speedY=-0.1}
   B=Multiply(A=TextureSample(UVs=@n8) {samplerType=SAMPLERTYPE_LinearGrayscale, texture=/Game/Effects/Fort_Effects/Emotes/UltraLight/Texture/T_Plasma_A_Greyscale}.R) {constB=0.5}
  )
  n10 = Multiply(
   A=@n6
   B=Multiply(
    A=TextureSample(UVs=@n9) {samplerType=SAMPLERTYPE_LinearGrayscale, texture=&T_Plasma_A_Greyscale}.R
    B=DynamicParameter() {paramNames [FalloffAdditionalBrightness, Hue, EnergyContrib, Bumpoffset]}.EnergyContrib
   )
  )
  n11 = Multiply(A=@n4.A, B=DepthFade(Opacity=Add(A=@n6, B=@n10)) {fadeDistanceDefault=10})
  MaterialFunctionCall(Emissive=StaticSwitchParameter(True=LinearInterpolate(B=@n5, Alpha=@n11) {constA=1}, False=@n5) {parameterName=bIsModulate}) {materialFunction=&MF_FX_EmissiveAdjustments}
  n12 = TextureSampleParameter2D(UVs=NamedRerouteUsage()) {
   group=Textures
   parameterName=T_MaskTexture
   samplerType=SAMPLERTYPE_Masks
   texture=/Game/Environments/Asteria/Sets/Ancients/Textures/T_Ancients_PRoom_Panel_A_M
  }
  NamedRerouteDeclaration(None=@n12.B) {name=TextureMask, nodeColor=<rgba 0.4766 0 1 1>}
  n13 = Power(Base=SphereMask(A=TextureCoordinate(), B=Constant() {r=0.5}) {attenuationRadius=0.5, hardnessPercent=0})
  n14 = DynamicParameter() {paramNames [FalloffAdditionalBrightness, Hue, EnergyContrib, Bumpoffset]}
  NamedRerouteDeclaration(None=StaticSwitchParameter(True=Constant(), False=Multiply(A=@n13, B=@n14.FalloffAdditionalBrightness) {constB=0.5}) {parameterName=bIsModulate}) {name=AdditionalBrightness, nodeColor=<rgba 0.2422 1 0 1>}
  NamedRerouteDeclaration(None=BumpOffset(
    Coordinate=TextureCoordinate()
    Height=DynamicParameter() {paramNames [FalloffAdditionalBrightness, Hue, EnergyContrib, Bumpoffset]}.Bumpoffset
   )) {name=ProjectedUV, nodeColor=<rgba 0.7578 0 1 1>}
  ParticlePositionWS()
 ]
}
