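# Niagara idle glow for an Ancient glyph in the Asteria environment set, tinted by User.CoreColor (HDR blue by default).
# BackingPlate bursts a fixed-facing sprite drawn by three renderers, with waveform-driven alpha flicker.
# CircularRays bursts three differently scaled ray meshes around the system's circle center, rotating with system age and animated through material parameters.
# A shared system Timeline fades both layers out through curves; the MeshRays emitter is disabled.
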NiagaraSystem /Game/Environments/Asteria/Sets/Ancients/FX/NS_AncientGlyph_Idle
dependencies {
 usedDataInterfaces [&NiagaraDataInterfaceCurve, &NiagaraDataInterfaceArrayFloat]
 usedDynamicInputs []
 usedModules [
  &self:SystemSpawnScript.SystemScriptSource.SystemScriptGraph.NiagaraNodeAssignment_1.SetVariables_#1
  &self:SystemSpawnScript.SystemScriptSource.SystemScriptGraph.NiagaraNodeAssignment_2.SetVariables_#2
  &self:SystemSpawnScript.SystemScriptSource.SystemScriptGraph.NiagaraNodeAssignment_3.SetVariables_#3
  &SystemState
  &Timeline
  &self:SystemSpawnScript.SystemScriptSource.SystemScriptGraph.NiagaraNodeAssignment_0.SetVariables_#4
 ]
}
emitters {
 BackingPlate {
  emitterSpawnScript [/SetVariables(Emitter.LightingFlicker=1) {isSetParameters=true}]
  emitterUpdateScript [
   &EmitterState()
   &SpawnBurst_Instantaneous(Use Spawn Probability=false, UseLoopCountLimit=false)
   /SetVariables(Emitter.LightingFlicker:NiagaraFloat=&Waveform("Waveform [1]"=::NewEnumerator0, "[1] Frequency"=0.5, "[1] Amplitude Min/Max"=<0.65 1>, "Waveform [2]"=::NewEnumerator1)) {isSetParameters=true}
   /SetVariables(Emitter.LightingFlicker:NiagaraFloat=&Multiply_Float(
     A=$Emitter.LightingFlicker
     B=&FloatFromCurve(FloatCurve=DataInterface() {propertyValues {Curve {keys [<at 0 1>,<at 1 0>]}}}, CurveIndex=$System.Timeline.NormalizedTime)
    )) {isSetParameters=true}
   /SetVariables(Emitter.AirglowCamOffset=140.3697) {isSetParameters=true}
  ]
  particleSpawnScript [
   &InitializeParticle(
    Color=&MultiplyLinearColor(Linear Color=$User.CoreColor, Scale Factor=<5 5 5 1>)
    Lifetime=$Emitter.CurrentLoopDuration
    Position Offset=<0 -35 576>
    Sprite Size Mode=::NewEnumerator3
    Uniform Sprite Size=1000
    UsePositionOffset=true
    Write Material Random=true
   )
   /SetVariables() {isSetParameters=true}
  ]
  particleUpdateScript [
   &ParticleState(Kill Particles When Lifetime Has Elapsed=false, Loop Particles Lifetime=true)
   &SolveForcesAndVelocity(Clamp Velocity=false, Limit Acceleration=false)
   /SetVariables(Particles.SpriteAlignment=<0 0 1>, Particles.SpriteFacing=<0 -1 0>) {isSetParameters=true}
   &ScaleColor(
    Scale Alpha=&Multiply_Float(A=&Multiply_Float(A=1, B=1), B=$Emitter.LightingFlicker)
    Scale RGB=&VectorFromFloat(Value=&Multiply_Float(
      A=0.0125
      B=&FloatFromCurve(FloatCurve=DataInterface() {propertyValues {Curve {keys [<at 0 1>,<at 1 0>]}}}, CurveIndex=$System.Timeline.NormalizedTime)
     ))
    ScaleA=true
    ScaleRGB=true
   )
   &CameraOffset()
   /SetVariables(Particles.ModulateColor:LinearColor=&MultiplyLinearColor(
     Linear Color=$Particles.Color
     Scale Factor=&MakeFloat4FromVectorAndFloat(
      "Vector (RGB)"=&VectorFromFloat(Value=10)
      "Float (Alpha)"=&Lerp_Float(A=1, B=3, Alpha=$Emitter.LightingFlicker)
     )
    )) {enabled=false, isSetParameters=true}
   /SetVariables(Particles.AirglowColor:LinearColor=&MultiplyLinearColor(Linear Color=$StackContext.Color, Scale Factor=<1 1 1 0>)) {enabled=false, isSetParameters=true}
   &DynamicMaterialParameters(
    Index 0 Param 1=&Multiply_Float(
     A=0.12
     B=&FloatFromCurve(FloatCurve=DataInterface() {propertyValues {Curve {keys [<at 0 1>,<at 1 0>]}}}, CurveIndex=$System.Timeline.NormalizedTime)
    )
    Index 0 Param 2=-0.1
    Index 0 Param 3=0.1
    Index 0 Param 4=0.1
    Module.Param0WriteEnabled=true
    Module.Param1WriteEnabled=true
    Module.Param2WriteEnabled=true
    Module.Param3WriteEnabled=true
   )
  ]
  rendererClasses [&NiagaraSpriteRendererProperties]
  renderers [
   {rendererClass=&NiagaraSpriteRendererProperties}
   {rendererClass=&NiagaraSpriteRendererProperties, rendererIndex=1}
   {rendererClass=&NiagaraSpriteRendererProperties, rendererIndex=2}
  ]
 }
 CircularRays {
  emitterUpdateScript [
   &EmitterState(Life Cycle Mode=::NewEnumerator1, Loop Behavior=::NewEnumerator1, Loop Duration Mode=::NewEnumerator1, UseLoopDelay=false)
   &SpawnRate(SpawnRate=33, Use Spawn Probability=false) {enabled=false}
   &SpawnBurst_Instantaneous(Spawn Count=3, Use Spawn Probability=false, UseLoopCountLimit=false)
  ]
  particleSpawnScript [
   &InitializeParticle(
    Color=&MultiplyLinearColor(Linear Color=$User.CoreColor, Scale Factor=<1 1 0.7 1>)
    Lifetime=5
    Mesh Scale Mode=::NewEnumerator3
    Mesh Uniform Scale=&SelectFloatFromArray(
     Direct Array Index=&ReturnExecIndex()
     Float Selection Array=DataInterface() {propertyValues {FloatData [1,1.3,1.8]}}
    )
    Position Offset=&AddVectorToPosition(Position=Linked() {linkedVariable {name=System.CircleCenter, type=Vector3f}}, Vector=<0 -20 0>)
    Sprite Size Mode=::NewEnumerator3
    Uniform Sprite Size=50
    UsePositionOffset=true
    Write Material Random=true
   )
   &InitialMeshOrientation(
    Mesh Axis To Orient=<0 0 1>
    Mesh Orientation Mode=::NewEnumerator2
    Orientation Vector=<0 -1 0>
    Rotation=&MakeVector(X=0, Y=0, Z=&RandomRangeFloat(Minimum=0, Maximum=250))
    Use Rotation Vector=true
   )
  ]
  particleUpdateScript [
   &ParticleState(Kill Particles When Lifetime Has Elapsed=false, Loop Particles Lifetime=true)
   &ScaleMeshSize(Scale Factor=&MakeVector(X=1, Y=1, Z=0.75))
   &ScaleColor(
    Scale Alpha=&FloatFromCurve(FloatCurve=DataInterface() {propertyValues {Curve {keys [<at 0 1>,<at 1 0>]}}}, CurveIndex=$System.Timeline.NormalizedTime, Scale Curve=0.22)
    Scale RGB=<0.25 0.25 0.25>
    ScaleA=true
    ScaleRGB=true
   )
   &SolveForcesAndVelocity(Clamp Velocity=false, Limit Acceleration=false)
   &DynamicMaterialParameters(
    Index 0 Param 1=3
    Index 0 Param 2=&Multiply_Float(A=$Engine.Time, B=0.05)
    Index 0 Param 3=530.6884
    Module.Param0WriteEnabled=true
    Module.Param1WriteEnabled=true
    Module.Param2WriteEnabled=true
    Module.Param3WriteEnabled=true
   )
   /SetVariables(Particles.MeshOrientation=&MultiplyQuaternion(
     Quaternion A=$Particles.Initial.MeshOrientation
     Quaternion B=&MakeQuatFromAxisAngle(
      Axis=&MakeVector(X=0, Y=0, Z=&RandomRangeFloat(Minimum=-1, Maximum=1))
      AngleInDegrees=$System.Age
     )
    )) {isSetParameters=true}
   &CameraOffset(Camera Offset Amount=211.994) {enabled=false}
  ]
  rendererClasses [&NiagaraMeshRendererProperties]
  renderers [{rendererClass=&NiagaraMeshRendererProperties, rendererIndex=0}]
 }
 MeshRays {
  bEnabled=false
  emitterUpdateScript [
   &EmitterState()
   &SpawnBurst_Instantaneous(Use Spawn Probability=false, UseLoopCountLimit=false)
  ]
  particleSpawnScript [&InitializeParticle(
    Color=&MultiplyLinearColor(Linear Color=$User.CoreColor, Scale Factor=<0.5 0.5 0.5 1>)
    Lifetime=5
    Mesh Scale Mode=::NewEnumerator3
    Mesh Uniform Scale=0.9
    Position Offset=$System.HeadDetail
    Sprite Size Mode=::NewEnumerator3
    Uniform Sprite Size=50
    UsePositionOffset=true
    Write Material Random=true
   )]
  particleUpdateScript [
   &ParticleState()
   &ScaleMeshSize(Scale Factor=&MakeVector(
     X=1
     Y=&FloatFromCurve(
      FloatCurve=DataInterface() {propertyValues {Curve {keys [
          <at 0 1 RCIM_Cubic>
          {arriveTangent=0.0388, interpMode=RCIM_Cubic, leaveTangent=0.0388, tangentMode=RCTM_User, time=0.1, value=1.25}
          <at 1 1.25 RCIM_Cubic>
         ]}}}
      CurveIndex=$System.Timeline.NormalizedTime
      Scale Curve=66.7686
     )
     Z=1
    ))
   &ScaleColor(
    Scale Alpha=&FloatFromCurve(
     FloatCurve=DataInterface() {propertyValues {Curve {keys [
         {arriveTangent=-2.0093, interpMode=RCIM_Cubic, leaveTangent=-2.0093, tangentMode=RCTM_User, time=-0.0046, value=1.0113}
         <at 1.0011 0.002 RCIM_Cubic>
        ]}}}
     Scale Curve=0.5
    )
    ScaleA=true
    ScaleRGB=false
   )
   &SolveForcesAndVelocity(Clamp Velocity=false, Limit Acceleration=false)
   &DynamicMaterialParameters(
    Index 0 Param 1=3
    Index 0 Param 2=&FloatFromCurve(FloatCurve=DataInterface() {propertyValues {Curve {keys [<at 0 1>,<at 1 0>]}}}, Scale Curve=0.1)
    Index 0 Param 3=&FloatFromCurve(
     FloatCurve=DataInterface() {propertyValues {Curve {keys [
         {arriveTangent=10.4947, interpMode=RCIM_Cubic, time=0, value=0}
         {arriveTangent=1.175, interpMode=RCIM_Cubic, leaveTangent=1.175, time=0.1, value=1.125}
         {interpMode=RCIM_Cubic, leaveTangent=-0.0405, time=1, value=1.175}
        ]}}}
     CurveIndex=&Multiply_Float(A=$System.Age, B=1.5477)
     Scale Curve=486.8016
    )
    Module.Param0WriteEnabled=true
    Module.Param1WriteEnabled=true
    Module.Param2WriteEnabled=true
    Module.Param3WriteEnabled=true
   )
   &CameraOffset(Camera Offset Amount=2.6204)
  ]
  rendererClasses [&NiagaraMeshRendererProperties]
  renderers [{rendererClass=&NiagaraMeshRendererProperties, rendererIndex=0}]
 }
}
summary {
 systemName=NS_AncientGlyph_Idle
 emitters [
  {bEnabled=false, emitterName=MeshRays, rendererClasses [&NiagaraMeshRendererProperties]}
  {emitterName=BackingPlate, rendererClasses [&NiagaraSpriteRendererProperties]}
  {emitterName=CircularRays, rendererClasses [&NiagaraMeshRendererProperties]}
 ]
 userVariables [
  User.CoreColor {type=LinearColor, description="", defaultValue {struct=LinearColor, value=<rgba 0.55 2.125 10 1>}}
  User.DesertBoulder {type=NiagaraBool}
  User.PhaseOut {type=NiagaraBool}
 ]
}
